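# Parent material for the Reboot Van sign: an opaque, lit, switch-heavy Fortnite base shader with light-cone logic.
# Up to six supported texture sets blend, then pass through internal decal, detail normal, damage, world-space color picker and reflection options.
# World position offset supports wind, cloth, tree and morph animation.
# Light-cone math combines noise, fresnel and depth fades; time-of-day intensity drives an offset collapsing vertices to the pivot at zero.
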Material /Game/Athena/Items/Gameplay/RebootVanRebuild/RebootVanReduxCards/Materials/M_RebootVan_Sign_PARENT
graph {
 properties {
  bAutomaticallySetUsageInEditor=false
  bUseMaterialAttributes=true
  bUsedWithNanite=true
  ditheredLODTransition=true
  pixelDepthOffsetMode=PDOM_Legacy
  refractionMethod=RM_IndexOfRefraction
  usedShadingModels=Default Lit
 }
 nodes [
  n1 = ComponentMask(None=TextureCoordinate() {coordinateIndex=1}) {g=true}
  n2 = ScalarParameter() {
   defaultValue=0.1
   desc="1/number of colors. ie 1/10 = .1"
   group=World Space Color Picker
   parameterName="Number Of Colors in Color Chart (1/Number of colors)"
  }
  n3 = Divide(A=ActorPositionWS(), B=ScalarParameter() {defaultValue=16.71, group=World Space Color Picker, parameterName="Color Change Frequency (WS)"})
  n4 = MaterialFunctionCall!(f2=ComponentMask(None=@n3) {g=true, r=true}, f3=@n3, f4) {materialFunction=/Game/Packages/Fortnite_UniqueMaterials_1/VectorComponentOps}
  n5 = StaticSwitchParameter(True=@n4.f3 Add, False=@n4.f3 Add) {defaultValue=true, group=World Space Color Picker, parameterName="Use XYZ(T) or XY(F) to Pick World Position Coloring Colors"}
  n6 = StaticSwitchParameter(True=ScalarParameter() {group=World Space Color Picker, parameterName=World Position Coloring Override Value}, False=@n5) {group=World Space Color Picker, parameterName=Override World Position Coloring Color Picker with a Float Value}
  n7 = StaticSwitchParameter(True=Add(A=Multiply(A=@n1, B=@n2), B=@n6), False=@n6) {group=World Space Color Picker, parameterName=Offset Color Index with UV Channel 2 G}
  AppendVector(A=@n7, B=@n7)
  ComponentMask() {g=true, r=true}
  Constant4Vector() {constant=<rgba 1 0 1 0>}
  MaterialFunctionCall() {materialFunction=&ComponentPivotLocation}
  n8 = ComponentMask(None=TextureCoordinate() {coordinateIndex=1}) {r=true}
  n9 = MaterialFunctionCall(
   X=SphereMask(A=@n8, B=Constant() {r=0.5}) {attenuationRadius=0.5}
   Y=SphereMask(A=@n8, B=Constant() {r=1.5}) {attenuationRadius=0.5}
   Z=SphereMask(A=@n8, B=Constant() {r=2.5}) {attenuationRadius=0.5}
   A=SphereMask(A=@n8, B=Constant() {r=3.5}) {attenuationRadius=0.5}
  ) {materialFunction=&MakeFloat4}
  n10 = MaterialFunctionCall!(Use UV 2=StaticBoolParameter() {group=Material_1, parameterName=Mat 1 Use UV2}) {materialFunction=/Game/Packages/Fortress_SharedMaterials/Base_Material_Functions/UseUV1or2}
  n11 = VectorParameter() {group=Diffuse, parameterName=Base Color Offset RG Tiling BA, defaultValue=<rgba 0 0 1 1>}
  n12 = ComponentMask(None=@n11.RGB) {g=true, r=true}
  n13 = StaticSwitchParameter(True=Multiply(A=Add(A=@n10.Result, B=@n12), B=AppendVector(A=@n11.B, B=@n11.B)), False=@n10.Result) {group=Diffuse, parameterName=Manipulate Base Color and Spec UVS}
  n14 = TextureSampleParameter2D(UVs=@n13) {
   group=Material_1
   parameterName=SpecularMasks
   samplerSource=SSM_Wrap_WorldGroupSettings
   samplerType=SAMPLERTYPE_Masks
   texture=&T_Fortnite_Default_S
  }
  n15 = VectorParameter() {group=Diffuse, parameterName=Base Color World Position Coordinate Scale, defaultValue=<rgba 256 256 256 1>}
  n16 = WorldPosition() {worldPositionShaderOffset=WPT_ExcludeAllShaderOffsets}
  n17 = MaterialFunctionCall(
   TextureObject=TextureObjectParameter() {desc=Use for exterior walls and floors, group=Material_1, parameterName=Diffuse}
   TextureSize=@n15.RGB
   WorldPosition=@n16.XYZ
   Export Float 4=StaticBool() {value=true}
  ) {materialFunction=&WorldAlignedTexture}
  n18 = StaticSwitchParameter(True=@n17.XY Texture, False=@n17.XY Texture) {group=Diffuse, parameterName=Project Base Color On XY or Z if false}
  n19 = TextureSampleParameter2D(UVs=@n13) {desc=Use for exterior walls and floors, group=Material_1, parameterName=Diffuse, samplerSource=SSM_Wrap_WorldGroupSettings}
  n20 = VectorParameter() {group=Diffuse, parameterName=Base Color, defaultValue=<rgba 0.5 0.5 0.5 1>}
  n21 = StaticSwitchParameter(
   True=StaticSwitchParameter(
    True=StaticSwitchParameter(True=@n17.XYZ Texture, False=@n18) {group=Diffuse, parameterName=Project Base Color on XYZ}
    False=AppendVector(A=@n19.RGB, B=@n19.RGB)
   ) {group=Diffuse, parameterName=Project Base Color Via World Position}
   False=AppendVector(A=@n20.RGB, B=@n20.RGB)
  ) {defaultValue=true, group=Diffuse, parameterName="Use a Base Color Texture "}
  n22 = VectorParameter() {group=LerpTexture, parameterName=Lerp Base Color World Position Coordinate Scale, defaultValue=<rgba 256 256 256 1>}
  n23 = VectorParameter() {group=LerpTexture, parameterName="Offset World Position Lerp  Base Color Coordinates", defaultValue=<rgba 0 0 0 1>}
  n24 = MaterialFunctionCall(
   TextureObject=TextureObjectParameter() {group=LerpTexture, parameterName=Base Color Lerp Texture}
   TextureSize=@n22.RGB
   WorldPosition=StaticSwitchParameter(True=Add(A=@n16.XYZ, B=@n23.RGB), False=@n16.XYZ) {group=LerpTexture, parameterName=Offset World Position Lerp Coordinates}
   Export Float 4=StaticBool() {value=true}
  ) {materialFunction=&WorldAlignedTexture}
  n25 = StaticSwitchParameter(True=TextureCoordinate() {coordinateIndex=1}, False=TextureCoordinate()) {group=LerpTexture, parameterName="Use UV channel 0(false) or 1(true) for the lerp texture"}
  n26 = VectorParameter() {group=LerpTexture, parameterName=Lerp Base Color UV Offset RG Tiling BA, defaultValue=<rgba 0 0 1 1>}
  n27 = Add(A=@n25, B=ComponentMask(None=@n26.RGB) {g=true, r=true})
  n28 = AppendVector(A=@n26.B, B=@n26.B)
  n29 = TextureSampleParameter2D(UVs=StaticSwitchParameter(True=Multiply(A=@n27, B=@n28), False=@n25) {group=LerpTexture, parameterName=Manipulate Lerp UVS}) {group=LerpTexture, parameterName=Base Color Lerp Texture, samplerSource=SSM_Wrap_WorldGroupSettings}
  n30 = StaticSwitchParameter(
   True=StaticSwitchParameter(
    True=@n24.XYZ Texture
    False=StaticSwitchParameter(True=@n24.XY Texture, False=@n24.XY Texture) {group=LerpTexture, parameterName=Lerp Texture_Project via WorldPosition XY or if false Z}
   ) {group=LerpTexture, parameterName=Lerp Texture_Project via WorldPosition XYZ}
   False=AppendVector(A=@n29.RGB, B=@n29.RGB)
  ) {group=LerpTexture, parameterName=Use World Position Projection for the Lerp Texture}
  n31 = ComponentMask(None=@n30) {b=true, g=true, r=true}
  n32 = Multiply(A=@n31, B=VectorParameter() {group=LerpTexture, parameterName=Multiply Base Color Lerp Texture, defaultValue=<rgba 1 1 1 1>}.RGB)
  n33 = TextureSampleParameter2D() {group=LerpTexture, parameterName=Height Lerp Height Texture, samplerSource=SSM_Wrap_WorldGroupSettings}
  n34 = StaticSwitchParameter(
   True=@n21
   False=StaticSwitchParameter(True=@n30, False=AppendVector(A=@n33.RGB, B=@n33.RGB)) {group=LerpTexture, parameterName=Use a Base Color Lerp Texture Channel as the Height Lerp Height Map}
  ) {group=LerpTexture, parameterName=Use a Base Color Channel as Height Lerp Height Map}
  n35 = VertexColor()
  n36 = AppendVector(A=@n35, B=@n35)
  n37 = StaticComponentMaskParameter(None=@n36) {defaultR=true, group=LerpTexture, parameterName=Vertex Color to Use For HeightLerp Falloff Contrast}
  n38 = ScalarParameter() {group=LerpTexture, parameterName=Height Lerp Transition Contrast}
  n39 = VectorParameter() {group=LerpTexture, parameterName=Offset World Position Lerp Alpha Coordinates, defaultValue=<rgba 0 0 0 1>}
  n40 = MaterialFunctionCall(
   TextureObject=TextureObjectParameter() {group=LerpTexture, parameterName=Lerp Alpha}
   TextureSize=VectorParameter() {group=LerpTexture, parameterName=WorldPositionCoordinates_LerpAlpha, defaultValue=<rgba 256 256 256 1>}.RGB
   WorldPosition=StaticSwitchParameter(True=Add(A=@n16.XYZ, B=@n39.RGB), False=@n16.XYZ) {group=LerpTexture, parameterName=Offset World Position Lerp Alpha Coordinates}
   Export Float 4=StaticBool() {value=true}
  ) {materialFunction=&WorldAlignedTexture}
  n41 = StaticSwitchParameter(True=TextureCoordinate() {coordinateIndex=1}, False=TextureCoordinate()) {group=LerpTexture, parameterName="Use UV channel 0(false) or 1(true) for the lerp texture alpha"}
  n42 = VectorParameter() {group=LerpTexture, parameterName=Lerp Alpha Offset RG Tiling BA, defaultValue=<rgba 0 0 1 1>}
  n43 = Add(A=@n41, B=ComponentMask(None=@n42.RGB) {g=true, r=true})
  n44 = AppendVector(A=@n42.B, B=@n42.B)
  n45 = TextureSampleParameter2D(UVs=StaticSwitchParameter(True=Multiply(A=@n43, B=@n44), False=@n41) {group=LerpTexture, parameterName=Manipulate Lerp Alpha UVS}) {group=LerpTexture, parameterName=Lerp Alpha, samplerSource=SSM_Wrap_WorldGroupSettings}
  n46 = StaticSwitchParameter(
   True=StaticSwitchParameter(
    True=@n40.XYZ Texture
    False=StaticSwitchParameter(True=@n40.XY Texture, False=@n40.XY Texture) {group=LerpTexture, parameterName=Lerp Alpha Texture_Project via WorldPosition XY or if false Z}
   ) {group=LerpTexture, parameterName=Lerp Alpha Texture_Project via WorldPosition XYZ}
   False=AppendVector(A=@n45.RGB, B=@n45.RGB)
  ) {group=LerpTexture, parameterName=Use World Pos Coordinates for Lerp Alpha Texture}
  n47 = StaticComponentMaskParameter(None=StaticSwitchParameter(
    True=@n21
    False=StaticSwitchParameter(True=@n30, False=@n46) {group=LerpTexture, parameterName="Use a Lerp  Base Color Channel for the Lerp Alpha"}
   ) {group=LerpTexture, parameterName=Use a Base Color Channel as a Lerp Alpha OR if false use a specific lerp alpha texture}) {defaultA=true, group=LerpTexture, parameterName=Texture Channel to use as the Lerp Alpha}
  n48 = ScalarParameter() {defaultValue=0.5, group=LerpTexture, parameterName=Lerp Alpha}
  n49 = StaticSwitchParameter(
   True=MaterialFunctionCall!(
    Lerp A
    Lerp B
    Height Texture=StaticComponentMaskParameter(None=@n34) {defaultA=true, group=LerpTexture, parameterName=Height Lerp Height Texture Channel}
    Falloff Contrast=StaticSwitchParameter(True=@n37, False=@n38) {group=LerpTexture, parameterName=Use Vertex Color for falloff contrast}
    Transition Phase=StaticComponentMaskParameter(None=@n36) {defaultG=true, group=LerpTexture, parameterName="Vertex Color to Use for HeightLerp Transition  Phase"}
   ) {materialFunction=/Game/Packages/Fortnight_MaterialFunctions/Temp/HeightLerp}.Lerp Alpha
   False=StaticSwitchParameter(
    True=StaticComponentMaskParameter(None=@n36) {defaultR=true, group=LerpTexture, parameterName=Vertex Color to Use as Lerp Alpha}
    False=StaticSwitchParameter(True=@n47, False=@n48) {group=LerpTexture, parameterName=Use Texture as Lerp Alpha or Scalar}
   ) {group=LerpTexture, parameterName=Use Vertex Color As Alpha for Lerp Texture}
  ) {group=LerpTexture, parameterName=Use HeightLerp Texture Blending}
  n50 = LinearInterpolate(
   A=ComponentMask(None=@n21) {b=true, g=true, r=true}
   B=StaticSwitchParameter(
    True=StaticSwitchParameter(True=@n32, False=@n31) {group=LerpTexture, parameterName=Modulate the Lerp Base Color}
    False=VectorParameter() {group=LerpTexture, parameterName=Lerp color, defaultValue=<rgba 0 0 0 1>}.RGB
   ) {defaultValue=true, group=LerpTexture, parameterName=Lerp using Texture or Vector}
   Alpha=@n49
  )
  n51 = ComponentMask(None=@n21) {a=true}
  n52 = MaterialFunctionCall(Use UV 2=StaticBoolParameter() {group=Material_1, parameterName=Mat 1 Use UV2}) {materialFunction=&UseUV1or2}
  n53 = VectorParameter() {group=Normal, parameterName=Normal Offset RG Tiling BA, defaultValue=<rgba 0 0 1 1>}
  n54 = Add(A=@n52.Result, B=ComponentMask(None=@n53.RGB) {g=true, r=true})
  n55 = AppendVector(A=@n53.B, B=@n53.B)
  n56 = TextureSampleParameter2D(UVs=StaticSwitchParameter(True=Multiply(A=@n54, B=@n55), False=@n52.Result) {group=Normal, parameterName=Manipulate Normal UVS}) {
   group=Material_1
   parameterName=Normals
   samplerSource=SSM_Wrap_WorldGroupSettings
   samplerType=SAMPLERTYPE_Normal
   texture=/Engine/EngineMaterials/FlatNormal
  }
  n57 = StaticSwitchParameter(True=@n56.RGB, False=Constant3Vector() {constant=<rgba 0 0 1 0>}) {group=Material_1, parameterName=UseNormalTexture}
  n58 = Constant3Vector() {constant=<rgba 0 0 1 0>}
  n59 = StaticSwitchParameter(
   True=LinearInterpolate(A=@n57, B=@n58, Alpha=ScalarParameter() {defaultValue=0.5, group=Normal, parameterName=Soften Normal Amount})
   False=@n57
  ) {group=Normal, parameterName=Soften Normals}
  n60 = MaterialFunctionCall(Use UV 2=StaticBoolParameter() {group=Normal, parameterName=Lerp Normal Use UV2}) {materialFunction=&UseUV1or2}
  n61 = VectorParameter() {group=Normal, parameterName=Lerp Normal Offset RG Tiling BA, defaultValue=<rgba 0 0 1 1>}
  n62 = Add(A=@n60.Result, B=ComponentMask(None=@n61.RGB) {g=true, r=true})
  n63 = AppendVector(A=@n61.B, B=@n61.B)
  n64 = TextureSampleParameter2D(UVs=StaticSwitchParameter(True=Multiply(A=@n62, B=@n63), False=@n60.Result) {group=Normal, parameterName=Manipulate Lerp Normal UVS}) {
   group=Normal
   parameterName=Second Normals for Lerp
   samplerSource=SSM_Wrap_WorldGroupSettings
   samplerType=SAMPLERTYPE_Normal
   texture=&FlatNormal
  }
  n65 = LinearInterpolate(A=@n64.RGB, B=@n58, Alpha=ScalarParameter() {defaultValue=0.5, group=Normal, parameterName=Soften Lerp Normal Amount})
  n66 = StaticSwitchParameter(
   True=LinearInterpolate(A=@n59, B=StaticSwitchParameter(True=@n65, False=@n64.RGB) {group=Normal, parameterName=Soften Lerp Normals}, Alpha=@n49)
   False=@n59
  ) {group=LerpTexture, parameterName=Lerp Normals}
  n67 = VectorParameter() {group=Normal, parameterName=Normal Modulation, defaultValue=<rgba 1 1 1 1>}
  n68 = MaterialFunctionCall(Use UV 2=StaticBoolParameter() {group=Multiple_Material_Textures, parameterName=Mat 2 Use UV2}) {materialFunction=&UseUV1or2}
  n69 = TextureSampleParameter2D(UVs=@n68.Result) {
   desc=Use for interior walls and ceilings
   group=Material_2
   parameterName=Normals_Texture_2
   samplerSource=SSM_Wrap_WorldGroupSettings
   samplerType=SAMPLERTYPE_Normal
   texture=&FlatNormal
  }
  n70 = TextureSampleParameter2D(UVs=@n68.Result) {
   desc=Use for interior walls and ceilings
   group=Material_2
   parameterName=Diffuse_Texture_2
   samplerSource=SSM_Wrap_WorldGroupSettings
   texture=/Engine/EngineMaterials/DefaultDiffuse
  }
  n71 = TextureSampleParameter2D(UVs=@n68.Result) {
   desc=Use for interior walls and ceilings
   group=Material_2
   parameterName=SpecularMasks_2
   samplerSource=SSM_Wrap_WorldGroupSettings
   samplerType=SAMPLERTYPE_Masks
   texture=&T_Fortnite_Default_S
  }
  n72 = MaterialFunctionCall(Use UV 2=StaticBoolParameter() {group=Multiple_Material_Textures, parameterName=Mat 3 Use UV2}) {materialFunction=&UseUV1or2}
  n73 = TextureSampleParameter2D(UVs=@n72.Result) {
   desc=Use for trims and accents
   group=Material_3
   parameterName=Diffuse_Texture_3
   samplerSource=SSM_Wrap_WorldGroupSettings
   texture=&DefaultDiffuse
  }
  n74 = TextureSampleParameter2D(UVs=@n72.Result) {
   desc=Use for trims and accents
   group=Material_3
   parameterName=SpecularMasks_3
   samplerSource=SSM_Wrap_WorldGroupSettings
   samplerType=SAMPLERTYPE_Masks
   texture=&T_Fortnite_Default_S
  }
  n75 = TextureSampleParameter2D(UVs=@n72.Result) {
   desc=Use for trims and accents
   group=Material_3
   parameterName=Normals_Texture_3
   samplerSource=SSM_Wrap_WorldGroupSettings
   samplerType=SAMPLERTYPE_Normal
   texture=&FlatNormal
  }
  n76 = MaterialFunctionCall(Use UV 2=StaticBoolParameter() {group=Multiple_Material_Textures, parameterName=Mat 4 Use UV2}) {materialFunction=&UseUV1or2}
  n77 = TextureSampleParameter2D(UVs=@n76.Result) {
   desc=Use for doors
   group=Material_4
   parameterName=Diffuse_Texture_4
   samplerSource=SSM_Wrap_WorldGroupSettings
   texture=&DefaultDiffuse
  }
  n78 = TextureSampleParameter2D(UVs=@n76.Result) {
   desc=Use for doors
   group=Material_4
   parameterName=SpecularMasks_4
   samplerSource=SSM_Wrap_WorldGroupSettings
   samplerType=SAMPLERTYPE_Masks
   texture=&T_Fortnite_Default_S
  }
  n79 = TextureSampleParameter2D(UVs=@n76.Result) {
   desc=Use for doors
   group=Material_4
   parameterName=Normals_Texture_4
   samplerSource=SSM_Wrap_WorldGroupSettings
   samplerType=SAMPLERTYPE_Normal
   texture=&FlatNormal
  }
  n80 = TextureSampleParameter2D() {group=Material_5, parameterName=Diffuse_Texture_5, samplerSource=SSM_Wrap_WorldGroupSettings, texture=&DefaultDiffuse}
  n81 = TextureSampleParameter2D() {
   group=Material_5
   parameterName=Normals_Texture_5
   samplerSource=SSM_Wrap_WorldGroupSettings
   samplerType=SAMPLERTYPE_Normal
   texture=&FlatNormal
  }
  n82 = TextureSampleParameter2D() {
   group=Material_5
   parameterName=SpecularMasks_5
   samplerSource=SSM_Wrap_WorldGroupSettings
   samplerType=SAMPLERTYPE_Masks
   texture=&T_Fortnite_Default_S
  }
  n83 = TextureSampleParameter2D() {group=Material_6, parameterName=Diffuse_Texture_6, samplerSource=SSM_Wrap_WorldGroupSettings, texture=&DefaultDiffuse}
  n84 = TextureSampleParameter2D() {
   group=Material_6
   parameterName=SpecularMasks_6
   samplerSource=SSM_Wrap_WorldGroupSettings
   samplerType=SAMPLERTYPE_Masks
   texture=&T_Fortnite_Default_S
  }
  n85 = TextureSampleParameter2D() {
   group=Material_6
   parameterName=Normals_Texture_6
   samplerSource=SSM_Wrap_WorldGroupSettings
   samplerType=SAMPLERTYPE_Normal
   texture=&FlatNormal
  }
  n86 = TextureSampleParameter2D() {
   group=z_DoNotUse_Deprecated
   parameterName=Diffuse_Texture_7
   samplerSource=SSM_Wrap_WorldGroupSettings
   texture=&DefaultDiffuse
  }
  n87 = TextureSampleParameter2D() {
   group=z_DoNotUse_Deprecated
   parameterName=Diffuse_Texture_8
   samplerSource=SSM_Wrap_WorldGroupSettings
   texture=&DefaultDiffuse
  }
  n88 = TextureSampleParameter2D() {
   group=z_DoNotUse_Deprecated
   parameterName=Diffuse_Texture_9
   samplerSource=SSM_Wrap_WorldGroupSettings
   texture=&DefaultDiffuse
  }
  n89 = TextureSampleParameter2D() {
   group=z_DoNotUse_Deprecated
   parameterName=Diffuse_Texture_10
   samplerSource=SSM_Wrap_WorldGroupSettings
   texture=&DefaultDiffuse
  }
  n90 = VectorParameter() {group=Diffuse, parameterName=Dot Base Color Channel, defaultValue=<rgba 1 0 0 0>}
  n91 = MaterialFunctionCall!(
   Spec1=@n14.RGB
   Texture1=StaticSwitchParameter(True=AppendVector(A=@n50, B=@n51), False=@n21) {group=LerpTexture, parameterName=Lerp Diffuse With Another Texture}
   Normal1=StaticSwitchParameter(True=Multiply(A=@n66, B=@n67.RGB), False=@n66) {group=Normal, parameterName=ModulateNormals}
   Normal2=@n69.RGB
   Texture2=AppendVector(A=@n70.RGB, B=@n70.RGB)
   Spec2=@n71.RGB
   Texture3=AppendVector(A=@n73.RGB, B=@n73.RGB)
   Spec3=@n74.RGB
   Normal3=@n75.RGB
   Texture4=AppendVector(A=@n77.RGB, B=@n77.RGB)
   Spec4=@n78.RGB
   Normal4=@n79.RGB
   Texture5=AppendVector(A=@n80.RGB, B=@n80.RGB)
   Normal5=@n81.RGB
   Spec5=@n82.RGB
   Texture6=AppendVector(A=@n83.RGB, B=@n83.RGB)
   Spec6=@n84.RGB
   Normal6=@n85.RGB
   Texture7=AppendVector(A=@n86.RGB, B=@n86.RGB)
   Texture8=AppendVector(A=@n87.RGB, B=@n87.RGB)
   Texture9=AppendVector(A=@n88.RGB, B=@n88.RGB)
   Use_Dot_Texture9
   Texture10=AppendVector(A=@n89.RGB, B=@n89.RGB)
   Use 2 Diffuse Textures=StaticBoolParameter() {group=Material_2, parameterName=Use 2 Materials}
   --------------
   Use 3 Diffuse Textures=StaticBoolParameter() {group=Material_3, parameterName=Use 3 Materials}
   Use 4 Diffuse Textures=StaticBoolParameter() {group=Material_4, parameterName=Use 4 Materials}
   Use 5 Diffuse Textures=StaticBoolParameter() {group=Material_5, parameterName=Use 5 Materials}
   Use 6 Diffuse Textures=StaticBoolParameter() {group=Material_6, parameterName=Use 6 Materials}
   Use 7 Diffuse Textures=StaticBoolParameter() {group=z_DoNotUse_Deprecated, parameterName=Use 7 Materials}
   Use 8 Diffuse Textures=StaticBoolParameter() {group=z_DoNotUse_Deprecated, parameterName=Use 8 Materials}
   Use 9 Diffuse Textures=StaticBoolParameter() {group=z_DoNotUse_Deprecated, parameterName=Use 9 Materials}
   Use 10 Diffuse Textures=StaticBoolParameter() {group=z_DoNotUse_Deprecated, parameterName=Use 10 Materials}
   "-------------- "
   Dot_Texture1=AppendVector(A=@n90.RGB, B=@n90.RGB)
   TransitionInfo
   Dot_Texture2
   Dot_Texture3
   Dot_Texture4
   Dot_Texture5
   Dot_Texture6
   Dot_Texture7
   Dot_Texture8
   Dot_Texture9
   Dot_Texture10
   Use_Dot_Texture1=StaticBoolParameter() {group=Diffuse, parameterName="Dot Texture 1 (Base Color)"}
   Use_Dot_Texture2
   Use_Dot_Texture3
   Use_Dot_Texture4
   Use_Dot_Texture5
   Use_Dot_Texture6
   Use_Dot_Texture7
   Use_Dot_Texture8
   Use_Dot_Texture10
  ) {materialFunction=/Game/Packages/Fortress_SharedMaterials/Base_Material_Functions/Fortnite_TextureSetLerp}
  n92 = MaterialFunctionCall!(
   Texture Group F4=@n9.Result
   Base Material Specular Masks=@n91.SpecularMasks
   Tangent Space Normal=@n91.Normals
   Decal Normal
   Diffuse=@n91.SpecularMasks
  ) {materialFunction=/Game/Packages/Fortress_SharedMaterials/Base_Material_Functions/InternalDecal}
  n93 = ComponentMask(None=@n92.Specular Masks) {desc=spec, r=true}
  Multiply(A=ScalarParameter() {defaultValue=0.5, group=Specular, parameterName=SpecularMult}, B=@n93)
  n94 = ScalarParameter() {group=Light Cone, parameterName=LightConeEmissiveDayMult, sortPriority=1}
  n95 = CollectionParameter() {
   collection=&FortniteMaterialParameters
   parameterName=FortniteActiveTimeOfDay
  }
  n96 = DotProduct(
   A=MaterialFunctionCall(X=@n94, Y=@n94, Z=@n94, A=ScalarParameter() {defaultValue=1, group=Light Cone, parameterName=LightConeEmissiveNightMult, sortPriority=2}) {materialFunction=&MakeFloat4}.Result
   B=@n95
  )
  n97 = RayTracingQualitySwitch(Normal=@n96, RayTraced=ScalarParameter() {defaultValue=1, group=Light Cone, parameterName=LightConeEmissiveLumenMult}) {desc="Override for Lumen "}
  n98 = VectorParameter() {group=Light Cone, parameterName=Emissive Color, defaultValue=<rgba 1.9479 2 0.7793 0.2569>}
  n99 = Saturate(None=@n96)
  n100 = TextureCoordinate()
  n101 = Constant2Vector() {g=-0.5, r=0.5}
  n102 = MaterialFunctionCall() {materialFunction=&LinearGradient}
  n103 = Multiply(A=ScalarParameter() {defaultValue=1, group=Light Cone, parameterName=Light Cone Falloff Factor A}, B=@n102.VGradient)
  n104 = Divide(
   A=ScalarParameter() {defaultValue=0.1, group=Light Cone, parameterName=Light Cone Falloff Factor B}
   B=LinearInterpolate(
    A=ScalarParameter() {defaultValue=0.05, group=Light Cone, parameterName=Light Cone InvSq Low Cutoff, sliderMax=1, sliderMin=0.0001}
    Alpha=Power(Base=@n103)
   ) {constA=0.05}
  ) {constA=0.1}
  n105 = OneMinus(None=@n102.VGradient)
  n106 = TextureSample(UVs=Panner(Coordinate=TextureCoordinate() {vTiling=0.1}) {speedX=0.005}) {texture=/Game/Global/Textures/Default/Noise/T_LightConeNoise}
  n107 = TextureSample(UVs=Panner(Coordinate=TextureCoordinate() {vTiling=0.05}) {speedX=0.002}) {texture=&T_LightConeNoise}
  n108 = PixelDepth()
  n109 = ScalarParameter() {defaultValue=256, group=Light Cone, parameterName=Light Cone Near Depth Fade}
  n110 = Multiply(
   A=Multiply(A=ScalarParameter() {defaultValue=0.0872, group=Light Cone, parameterName=Light Cone Opacity}, B=Add(A=@n106.G, B=@n107.R))
   B=Clamp(None=Divide(A=@n108.R, B=@n109))
  )
  n111 = ScalarParameter() {defaultValue=64, group=Light Cone, parameterName=Light Cone Depth Fade}
  n112 = ComponentMask(None=@n100) {g=true}
  n113 = LinearInterpolate(
   A=Power(Base=SphereMask(A=@n100, B=@n101) {attenuationRadius=0.45, hardnessPercent=0}) {constExponent=4}
   B=Multiply(A=Multiply(A=@n104, B=@n105), B=DepthFade(Opacity=@n110, FadeDistance=@n111))
   Alpha=Saturate(None=Ceil(None=@n112)) {desc="Mask out the flare part of the mesh, which has UVs above the 0...1 space"}
  )
  n114 = MaterialFunctionCall(Normal Vector=VertexNormalWS(), Clamp Fresnel Dot Product=StaticBool()) {materialFunction=&Fresnel_Function}
  n115 = Power(Base=OneMinus(None=Abs(None=@n114.Result)), Exp=ScalarParameter() {defaultValue=2.2386, group=Light Cone, parameterName=Fresnel Power}) {constExponent=6}
  n116 = Power(
   Base=Clamp(None=Multiply(A=@n113, B=@n115))
   Exp=ScalarParameter() {defaultValue=1, group=Light Cone, parameterName=Light cone Opacity Power}
  )
  n117 = TextureSampleParameter2D(UVs=Multiply(A=TextureCoordinate(), B=ScalarParameter() {defaultValue=1, group=CustomDamage, parameterName=Custom Damage Height Texture Tiling})) {
   group=CustomDamage
   parameterName=Custom Damage Normal
   samplerSource=SSM_Wrap_WorldGroupSettings
   samplerType=SAMPLERTYPE_Normal
   texture=/Game/Global/Textures/Default/Tileable_Patterns/CarDamageNormal
  }
  n118 = TextureCoordinate()
  n119 = AppendVector(
   A=ScalarParameter() {defaultValue=1, group=DetailNormalMap, parameterName=DetailNormalMapTilingX}
   B=ScalarParameter() {defaultValue=1, group=DetailNormalMap, parameterName=DetailNormalMapTilingY}
  )
  n120 = TextureSampleParameter2D(UVs=StaticSwitchParameter(True=Multiply(A=@n118, B=@n119), False=@n118) {group=DetailNormalMap, parameterName=Manipulate DetailMap UVS}) {
   group=DetailNormalMap
   parameterName=DetailNormalMap
   samplerSource=SSM_Wrap_WorldGroupSettings
   samplerType=SAMPLERTYPE_Normal
   texture=&FlatNormalBlue
  }
  n121 = ScalarParameter() {defaultValue=1, group=DetailNormalMap, parameterName=DetailNormalStrength}
  n122 = AppendVector(A=AppendVector(A=@n121, B=@n121), B=Constant())
  n123 = StaticSwitchParameter(True=Add(A=@n92.Normals, B=Multiply(A=@n120.RGB, B=@n122)), False=@n92.Normals) {desc=Tangent Normal, group=DetailNormalMap, parameterName=DetailNormalMap}
  n124 = MaterialFunctionCall() {materialFunction=/Engine/Functions/Engine_MaterialFunctions02/WorldPositionOffset/StaticMeshMorphTargets}
  n125 = ScalarParameter() {defaultValue=1, group=MorphTargets, parameterName=MorphTargetStrength}
  n126 = StaticSwitchParameter(
   True=StaticSwitchParameter(True=LinearInterpolate(A=@n123, B=@n124.Morph Target 1 Normals, Alpha=@n125), False=@n123) {group=MorphTargets, parameterName=MorphNormals}
   False=@n123
  ) {group=MorphTargets, parameterName=UseMorphTargets}
  n127 = MaterialFunctionCall(BaseNormal=@n126, AdditionalNormal=@n117.RGB) {materialFunction=&BlendAngleCorrectedNormals}
  n128 = StaticSwitchParameter(
   True=MaterialFunctionCall(Speed=Constant() {r=0.3333}) {materialFunction=&TimeWithSpeedVariable}.Frac Time
   False=ScalarParameter() {group=z_code, parameterName=Code_CustomDamagePercentage}
  ) {group=CustomDamage, parameterName=Debug custom damage animation}
  n129 = ScalarParameter() {defaultValue=0.9, group=CustomDamage, parameterName=Custom Damage N Height Max}
  n130 = TextureSampleParameter2D() {
   group=CustomDamage
   parameterName=Custom Damage Height Map
   samplerSource=SSM_Wrap_WorldGroupSettings
   samplerType=SAMPLERTYPE_Masks
   texture=/Game/Packages/Fortress_ShaderTests/Textures/T_CopCar_DamageMask
  }
  n131 = StaticComponentMaskParameter(None=AppendVector(A=@n130.RGB, B=@n130.RGB)) {defaultG=true, group=CustomDamage, parameterName=Custom Damage Height Map Channel}
  n132 = ScalarParameter() {defaultValue=10, group=CustomDamage, parameterName=Custom Damage Normal Map Height Map Final Reveal Mask Multiplier}
  n133 = Multiply(
   A=MaterialFunctionCall(Transition Phase=Multiply(A=@n128, B=@n129), Height Texture=@n131) {materialFunction=/Engine/Functions/Engine_MaterialFunctions02/Texturing/HeightLerp}.Lerp Alpha No Contrast
   B=Clamp(None=Multiply(A=@n131, B=@n132) {constB=10})
  )
  n134 = StaticSwitchParameter(
   True=StaticSwitchParameter(True=LinearInterpolate(A=@n126, B=@n127.Result, Alpha=@n133), False=@n126) {desc=N, group=CustomDamage, parameterName=Use Custom Damage Normal Map}
   False=@n126
  ) {desc=Normal, group=CustomDamage, parameterName=Use Custom Damage}
  n135 = MaterialFunctionCall!(
   UVs
   Pan Speed
   WPO_ClothOffsetDirection
   WPO_ClothNormalIntensity
   WPO_ClothOffsetIntensity
   "Motion Mask (Defaults to the Red Vertex Color)"
   WPO_ClothWPOWindTextureOffset
   Pan Direction
  ) {materialFunction=/Game/Packages/Fortress_SharedMaterials/Base_Material_Functions/Cloth}
  n136 = MaterialFunctionCall!(
   Tangent Space Normal=StaticSwitchParameter(
    True=MaterialFunctionCall(BaseNormal=@n134, AdditionalNormal=@n135.Normal) {materialFunction=&BlendAngleCorrectedNormals}.Result
    False=@n134
   ) {group=WPO_ClothAnimation, parameterName=UseWPO_ClothAnimation}
   UseWind=StaticBoolParameter() {group=Wind, parameterName=UseWind}
  ) {materialFunction=/Game/Packages/Fortress_SharedMaterials/Base_Material_Functions/BaseShaderWindFunction}
  n137 = NamedRerouteUsage()
  n138 = Add(
   A=@n137
   B=MaterialFunctionCall!(Use Tree Vertex Shader=StaticBoolParameter() {group=Special Cases, parameterName=Use Tree Vertex Shader}) {materialFunction=/Game/Packages/Fortress_SharedMaterials/Base_Material_Functions/FortniteTreeAnimation_VS}.Result
  )
  n139 = MaterialFunctionCall() {materialFunction=&StaticMeshMorphTargets}
  n140 = ScalarParameter() {group=MorphTargets, parameterName=MorphTargetStrength}
  n141 = StaticSwitchParameter(True=Add(A=@n138, B=Multiply(A=@n139.Morph Target 1 WPO, B=@n140)) {desc=7 instructions}, False=@n138) {group=MorphTargets, parameterName=UseMorphTargets}
  n142 = VectorParameter() {group=WorldPositionOffset, parameterName=WorldPositionOffset_ModulatedByVertexColor, defaultValue=<rgba 0 0 0 1>}
  n143 = VertexColor()
  n144 = Multiply(
   A=@n142.RGB
   B=StaticComponentMaskParameter(None=AppendVector(A=@n143, B=@n143)) {defaultA=true, group=WorldPositionOffset, parameterName=WorldPositionOffsetVertexColorModulationChannel}
  )
  n145 = ShadingPathSwitch(
   Default=StaticSwitchParameter(True=Add(A=@n141, B=@n144), False=@n141) {group=WorldPositionOffset, parameterName=UseWorldPositionOffsetModulatedByVertexColors}
   Mobile=@n137
  ) {desc=Remove Some WPO on Mobile}
  n146 = MaterialFunctionCall() {materialFunction=&Cloth}
  n147 = Add(
   A=@n136.World position offset
   B=StaticSwitchParameter(True=Add(A=@n145, B=@n146.Additive WPO), False=@n145) {group=WPO_ClothAnimation, parameterName=UseWPO_ClothAnimation}
  )
  n148 = ScalarParameter() {desc="Try to use small values. ", group=WorldPositionOffset, parameterName=Camera Offset}
  n149 = WorldPosition()
  n150 = MaterialFunctionCall(Vector=Subtract(A=CameraPositionWS(), B=Add(A=@n147, B=@n149.XYZ))) {materialFunction=&SafeNormalize}
  n151 = StaticSwitchParameter(True=Add(A=@n147, B=Multiply(A=@n148, B=@n150.Result)), False=@n147) {group=WorldPositionOffset, parameterName=Use Camera Offset}
  n152 = Multiply(A=@n151)
  n153 = Constant()
  n154 = Subtract(
   A=MaterialFunctionCall() {materialFunction=&ComponentPivotLocation}.Result
   B=Add(A=@n152, B=WorldPosition().XYZ)
  )
  n155 = If(
   A=Multiply(A=@n96, B=ScalarParameter() {defaultValue=0.0872, group=Light Cone, parameterName=Light Cone Opacity})
   B=@n153
   "A > B"=@n153
   "A == B"=@n154
   "A < B"=@n154
  )
  n156 = VectorParameter() {group=Snow, parameterName=SnowColor, defaultValue=<rgba 0.7262 0.7933 0.83 1>}
  n157 = ScalarParameter() {group=Snow, parameterName=SnowNormalContrast1}
  n158 = ScalarParameter() {defaultValue=1, group=Snow, parameterName=SnowNormalContrast2}
  n159 = StaticSwitch(
   True=Multiply(A=@n136.Tangent Space Pixel Normal, B=TwoSidedSign())
   False=@n136.Tangent Space Pixel Normal
   Value=StaticBoolParameter() {group=Normal, parameterName=Use Two Sided Normals}
  )
  n160 = ComponentMask(None=Transform(None=@n159)) {b=true}
  n161 = ScalarParameter() {defaultValue=0.3802, group=Snow, parameterName=SnowDiffuseContrast1}
  n162 = ScalarParameter() {defaultValue=8.5895, group=Snow, parameterName=SnowDiffuseContrast2}
  n163 = ComponentMask(None=TextureCoordinate() {coordinateIndex=1}) {r=true}
  n164 = ActorPositionWS()
  n165 = TextureSampleParameter2D(UVs=Divide(
    A=MaterialFunctionCall(f3=@n164) {materialFunction=&VectorComponentOps}.f3 Add
    B=ScalarParameter() {defaultValue=16.71, group=Emissive, parameterName="Emissive Color Change Frequency (WS)"}
   )) {group=Emissive, parameterName=Emissive Color Chart, texture=/Game/Environments/Sets/Arid/Textures/T_EmissiveColorChart}
  n166 = VectorParameter() {group=Emissive, parameterName=Emissive Multiplier, defaultValue=<rgba 2 2 2 1>}
  n167 = VectorParameter() {group=Emissive, parameterName="EmissiveUVPositioning (RG)UpperLeft (BA)LowerRight", defaultValue=<rgba 0 0 1 1>}
  n168 = TextureSampleParameter2D(UVs=Multiply(A=TextureCoordinate(), B=ScalarParameter() {defaultValue=1.5, group=CustomDamage, parameterName=Custom Damage Texture Tiling})) {
   group=CustomDamage
   parameterName=Custom Damage Diffuse Texture
   samplerSource=SSM_Wrap_WorldGroupSettings
   texture=/Game/Packages/Fortress_ShaderTests/Textures/charcoal
  }
  n169 = TextureSampleParameter2D(UVs=@n7) {
   group=World Space Color Picker
   parameterName=Color Chart
   samplerSource=SSM_Wrap_WorldGroupSettings
   texture=/Game/Global/Textures/Default/ColorSets/CarColors_Interiors
  }
  n170 = MaterialFunctionCall!(Use Gradient, Materials masked per float 4 channel=@n9.Result, Diffuse=@n92."Base Color (Float 4)") {materialFunction=/Game/Packages/Fortress_SharedMaterials/Base_Material_Functions/WorldSpaceGradientFunction}
  n171 = ComponentMask(None=@n170.Out) {b=true, g=true, r=true}
  n172 = Multiply(A=@n169.RGB, B=@n171)
  n173 = TextureSampleParameter2D(UVs=@n7) {
   group=World Space Color Picker
   parameterName=Color Chart 2
   samplerSource=SSM_Wrap_WorldGroupSettings
   texture=/Game/Global/Textures/Default/ColorSets/CarColors
  }
  n174 = TextureSampleParameter2D() {
   group=World Space Color Picker
   parameterName=World Space Color Picker Mask
   samplerSource=SSM_Wrap_WorldGroupSettings
   texture=/Game/Vehicles/FORT_Vehicles_01/Materials/M_FORT_Truck_01_M_New
  }
  n175 = AppendVector(A=@n174.RGB, B=@n174.RGB)
  n176 = StaticComponentMaskParameter(None=@n175) {defaultB=true, group=World Space Color Picker, parameterName=World Space Color Picker Mask Channel}
  n177 = LinearInterpolate(
   A=LinearInterpolate(A=@n172, B=Multiply(A=@n173.RGB, B=@n171), Alpha=@n176)
   B=@n171
   Alpha=StaticComponentMaskParameter(None=@n175) {defaultG=true, group=World Space Color Picker, parameterName=World Space Color Picker 2 Mask Channel}
  )
  n178 = LinearInterpolate(A=@n172, B=@n171, Alpha=@n176)
  n179 = StaticSwitchParameter(
   True=StaticSwitchParameter(
    True=StaticSwitchParameter(True=@n177, False=@n178) {group=World Space Color Picker, parameterName=Use 2 World Position Coloring Colors}
    False=@n172
   ) {group=World Space Color Picker, parameterName=Mask World Position Color Picker}
   False=@n171
  ) {group=World Space Color Picker, parameterName=Use World Position Coloring}
  n180 = ScalarParameter() {defaultValue=1, group=Diffuse, parameterName=Base Color Desaturation Amount}
  n181 = ComponentMask(None=StaticSwitchParameter(True=Desaturation(None=@n179, Fraction=@n180), False=@n179) {group=Diffuse, parameterName=Desaturate Base Color}) {b=true, desc=diffuse output, g=true, r=true}
  n182 = TextureCoordinate()
  MaterialExpressionObjectWorldPosition_0 = ObjectPositionWS()
  n183 = Divide(
   A=AppendVector(
    A=MaterialFunctionCall(f2=@MaterialExpressionObjectWorldPosition_0) {materialFunction=&VectorComponentOps}.f2 Add
    B=Constant()
   )
   B=ObjectRadius()
  )
  n184 = ScalarParameter() {defaultValue=1, group=ModulateTexture, parameterName=ModTextureWorldPosOffsetRate}
  n185 = StaticSwitchParameter(True=Add(A=@n182, B=Multiply(A=@n183, B=@n184)), False=@n182) {defaultValue=true, group=ModulateTexture, parameterName=WorldPosOffsetModulateTexture}
  n186 = VectorParameter() {group=ModulateTexture, parameterName=ModTexture RG_XYScaling B_AddToTexture A_MultiplyTexture, defaultValue=<rgba 1 1 0 0>}
  n187 = ComponentMask(None=@n186.RGB) {g=true, r=true}
  n188 = TextureSampleParameter2D(UVs=Add(A=Multiply(A=@n185, B=@n187), B=AppendVector(A=@n186.B, B=@n186.B))) {group=ModulateTexture, parameterName=Modulate Base Color Texture, samplerSource=SSM_Wrap_WorldGroupSettings}
  n189 = StaticComponentMaskParameter(None=AppendVector(A=@n188.RGB, B=@n188.RGB)) {defaultB=true, defaultG=true, defaultR=true, group=ModulateTexture, parameterName=Modulate Base Color Texture Channel}
  n190 = StaticSwitchParameter(True=Multiply(A=@n181, B=@n189), False=@n181) {desc="< Diffuse", group=ModulateTexture, parameterName=Texture Modulate Base Color}
  n191 = VectorParameter() {group=Diffuse, parameterName=Base Color Multiplier, defaultValue=<rgba 1 1 1 1>}
  n192 = VertexColor()
  n193 = LinearInterpolate(
   A=VectorParameter() {group=Diffuse, parameterName=Base Color Multiplier_2, defaultValue=<rgba 1 1 1 1>}.RGB
   B=@n191.RGB
   Alpha=Clamp(None=StaticComponentMaskParameter(None=@n192) {defaultR=true, group=Diffuse, parameterName=ModuleBaseColorChanelMask})
  )
  n194 = StaticSwitchParameter(
   True=Multiply(A=@n190, B=StaticSwitchParameter(True=@n193, False=@n191.RGB) {group=Diffuse, parameterName=Modulate BasewithVertexColorMask})
   False=@n190
  ) {group=Diffuse, parameterName=Modulate Base Color}
  n195 = ScalarParameter() {defaultValue=1, group=reflection, parameterName=Reflection Fresnel Multiplier}
  n196 = StaticSwitchParameter(True=@n123, False=@n92.Normals) {group=DetailNormalMap, parameterName=Use Detail Map in Fresnel}
  n197 = StaticSwitchParameter(
   True=LinearInterpolate(
    A=VertexNormalWS()
    B=StaticSwitchParameter(True=@n196, False=@n123) {group=DetailNormalMap, parameterName=DetailNormalMap}
    Alpha=ScalarParameter() {defaultValue=1, group=reflection, parameterName=Reflection Fresnel Smoothness}
   )
   False=@n123
  ) {group=reflection, parameterName=Smooth Reflection Fresnel}
  n198 = CameraVectorWS()
  n199 = Clamp(None=DotProduct(A=Transform(None=@n197), B=@n198))
  n200 = DotProduct(A=@n198, B=VertexNormalWS())
  n201 = Multiply(
   A=Power(Base=OneMinus(None=@n199), Exp=ScalarParameter() {defaultValue=2, group=reflection, parameterName=Reflection Fresnel Power})
   B=OneMinus(None=Clamp(None=@n200)) {desc=should probably remove}
  )
  n202 = Add(
   A=StaticSwitchParameter(True=Multiply(A=@n195, B=@n201), False=@n201) {group=reflection, parameterName=Modulate Reflection Fresnel}
   B=ScalarParameter() {defaultValue=1, group=Normal, parameterName=Reflection Fresnel Offset}
  )
  n203 = Constant() {r=1}
  n204 = ReflectionVectorWS()
  n205 = Transform(None=ReflectionVectorWS()) {transformSourceType=TRANSFORMSOURCE_World, transformType=TRANSFORM_View}
  n206 = Multiply(A=ConstantBiasScale(None=ComponentMask(None=@n205) {g=true, r=true}), B=Constant2Vector() {g=-1, r=1})
  n207 = Multiply(
   A=Multiply(
    A=StaticSwitchParameter(True=@n202, False=@n203) {group=reflection, parameterName=Use Reflection Fresnel}
    B=ScalarParameter() {defaultValue=1, group=reflection, parameterName=Reflection Multiplier}
   )
   B=StaticSwitchParameter(
    True=TextureSampleParameterCube!(UVs=@n204, Apply View MipBias) {group=reflection, parameterName=Reflection CubeMap, samplerSource=SSM_Wrap_WorldGroupSettings}.RGB
    False=TextureSampleParameter2D(UVs=@n206) {group=reflection, parameterName=SphereReflectionMap, samplerSource=SSM_Wrap_WorldGroupSettings}.RGB
   ) {group=reflection, parameterName=CubeMap Or ReflectionProbe}
  ) {desc=Image * multiplier * fresnel}
  n208 = TextureCoordinate()
  n209 = VectorParameter() {group=LerpTexture, parameterName=ReflectionMask Offset RG Tiling BA, defaultValue=<rgba 0 0 1 1>}
  n210 = Add(A=@n208, B=ComponentMask(None=@n209.RGB) {g=true, r=true})
  n211 = AppendVector(A=@n209.B, B=@n209.B)
  n212 = TextureSampleParameter2D(UVs=StaticSwitchParameter(True=Multiply(A=@n210, B=@n211), False=@n208) {group=reflection, parameterName=Manipulate ReflectionMask UVS}) {group=reflection, parameterName=ReflectionMask, samplerSource=SSM_Wrap_WorldGroupSettings}
  n213 = AppendVector(A=@n212.RGB, B=@n212.RGB)
  n214 = StaticSwitchParameter(True=@n93, False=@n213) {group=Specular, parameterName=Use Spec as Reflection Mask}
  n215 = StaticComponentMaskParameter(None=StaticSwitchParameter(True=@n190, False=StaticSwitchParameter(True=@n214, False=@n213) {defaultValue=true, group=Specular, parameterName=Use Spec}) {group=Specular, parameterName=Use Diffuse as Reflection Mask}) {defaultR=true, group=reflection, parameterName=ReflectionMaskChannel}
  n216 = MaterialFunctionCall(In=@n215, Contrast=ScalarParameter() {group=reflection, parameterName=Reflection Mask Cheap Contrast}) {materialFunction=&CheapContrast}
  n217 = StaticSwitchParameter(
   True=StaticSwitchParameter(True=Clamp(None=@n216.Result), False=@n215) {group=reflection, parameterName=Apply Cheap Contrast To Reflection Mask}
   False=ScalarParameter() {defaultValue=0.5, group=reflection, parameterName=Reflection Amount}
  ) {group=reflection, parameterName=Mask Reflection Map With Image}
  n218 = Multiply(A=@n207, B=@n217)
  n219 = StaticSwitchParameter(
   True=StaticSwitchParameter(True=LinearInterpolate(A=@n194, B=@n207, Alpha=@n217) {desc="a=0 b=1"}, False=Add(A=@n194, B=@n218)) {group=reflection, parameterName=Lerp Blend Or Add Reflection}
   False=@n194
  ) {desc=Diffuse, group=reflection, parameterName=Use Reflection Map}
  n220 = TextureObjectParameter() {
   group=Material_Layer_TopBottom
   parameterName=Mask_Mat2
   samplerType=SAMPLERTYPE_Masks
   texture=&T_Fortnite_Default_S
  }
  n221 = ScalarParameter() {defaultValue=658.5729, group=Material_Layer_TopBottom, parameterName=TextureSize_Mat2}
  n222 = TextureObjectParameter() {group=Material_Layer_TopBottom, parameterName=Mask_Mat2_1, samplerType=SAMPLERTYPE_Normal, texture=&FlatNormal}
  n223 = MaterialFunctionCall!(
   Diffuse Texture Object=TextureObjectParameter() {group=Material_Layer_TopBottom, parameterName=BaseColor_Mat2, texture=&DefaultDiffuse}
   Spec Texture Object=@n220
   TextureSize=AppendVector(A=AppendVector(A=@n221, B=@n221), B=@n221)
   --------------------------
   Use seperate settings for Z projection
   Use seperate texture for Z projection
   Texture Size Z projection
   Z Plane Diffuse Texture Object
   Z Plane Spec Texture Object
   "--------------------------      "
   ProjectionTransitionContrast
   Use Texture Alpha Heightmap to Refine the Transistion
   HeightLerpTransitionContrast
   PickTheDiffuseHeightChannel
   TheDiffuseHeightChannel
   "--------------------------       "
   NormalTextureObject=@n222
   World Space Normals
   "--------------------------  "
   Export Float 4 Diffuse
   Export Float 4 Spec
   "--------------------------     "
   WorldPosition
   World Space Normal
  ) {materialFunction=/Engine/Functions/Engine_MaterialFunctions02/Texturing/WorldAlignedTextures_Complex}
  n224 = Multiply(A=TextureCoordinate(), B=ScalarParameter() {defaultValue=2, group=Material_Layer_TopBottom, parameterName=Scale_Mat2})
  n225 = TextureSampleParameter2D(UVs=@n224) {
   group=Material_Layer_TopBottom
   parameterName=BaseColor_Mat2
   samplerSource=SSM_Wrap_WorldGroupSettings
   texture=&DefaultDiffuse
  }
  n226 = StaticSwitchParameter(True=@n223.Z Diffuse Texture, False=@n225.RGB) {group=Material_Layer_TopBottom, parameterName=MatLayer2_isWorldAligned}
  n227 = TextureSampleParameter2D(UVs=@n224) {
   group=Material_Layer_TopBottom
   parameterName=Mask_Mat2
   samplerType=SAMPLERTYPE_Masks
   texture=&T_Fortnite_Default_S
  }
  n228 = MaterialFunctionCall(Float3=StaticSwitchParameter(True=@n223.Z Spec, False=@n227.RGB) {group=Material_Layer_TopBottom, parameterName=MatLayer2_isWorldAligned}) {materialFunction=&BreakOutFloat3Components}
  n229 = LinearInterpolate(
   A=ScalarParameter() {defaultValue=0.2, group=Material_Layer_TopBottom, parameterName=Specular_low_Mat2}
   B=ScalarParameter() {defaultValue=0.5, group=Material_Layer_TopBottom, parameterName=Specular_High_Mat2}
   Alpha=@n228.R
  )
  n230 = Multiply(A=@n228.B, B=ScalarParameter() {defaultValue=1, group=Material_Layer_TopBottom, parameterName=Roughness_Mat2})
  n231 = TextureSampleParameter2D(UVs=@n224) {group=Material_Layer_TopBottom, parameterName=Normal_Mat2, samplerType=SAMPLERTYPE_Normal, texture=&FlatNormal}
  n232 = StaticSwitchParameter(True=@n223.Normal XYZ Texture, False=@n231.RGB) {group=Material_Layer_TopBottom, parameterName=MatLayer2_isWorldAligned}
  n233 = VectorParameter() {group=Material_Layer_TopBottom, parameterName=Color_Tint_Mat2_B, defaultValue=<rgba 0.985 0.5687 0.4198 1>}
  n234 = VectorParameter() {group=Material_Layer_TopBottom, parameterName=Color_Tint_Mat2_A, defaultValue=<rgba 0.985 0.5687 0.4198 1>}
  n235 = ScalarParameter() {defaultValue=-0.2657, group=Material_Layer_TopBottom, parameterName=TintMask_offset_Mat2}
  n236 = ScalarParameter() {defaultValue=-0.2047, group=Material_Layer_TopBottom, parameterName=DirectionBlend_Height}
  n237 = Multiply(A=ComponentMask(None=VertexNormalWS()) {b=true}, B=Clamp(None=Add(A=@n228.G, B=@n236)))
  n238 = OneMinus(None=@n237)
  n239 = Clamp(None=MaterialFunctionCall(
    In=Add(A=@n235, B=@n238)
    Contrast=ScalarParameter() {defaultValue=4.0435, group=Material_Layer_TopBottom, parameterName=TintMask_Contrast_Mat2}
   ) {materialFunction=&CheapContrast}.Result)
  n240 = ScalarParameter() {defaultValue=-0.0459, group=Material_Layer_TopBottom, parameterName=DirectionBlend_Offset}
  n241 = Clamp(None=LinearInterpolate(
    A=ScalarParameter() {defaultValue=-1.2699, group=Material_Layer_TopBottom, parameterName=DirectionBlend_low}
    B=ScalarParameter() {defaultValue=12.6981, group=Material_Layer_TopBottom, parameterName=DirectionBlend_High}
    Alpha=Add(A=@n237, B=@n240)
   ))
  n242 = WorldPosition()
  n243 = ComponentMask(None=ActorPositionWS()) {b=true}
  n244 = Constant() {r=384}
  n245 = Frac(None=Divide(A=@n243, B=@n244))
  n246 = Constant() {r=0.5}
  n247 = Add(A=@n243, B=Multiply(A=OneMinus(None=@n245), B=@n244))
  n248 = Subtract(A=@n243, B=Multiply(A=@n245, B=@n244))
  n249 = Divide(
   A=Subtract(A=ComponentMask(None=@n242.XYZ) {b=true}, B=If(A=@n245, B=@n246, "A > B"=@n247, "A == B"=@n248, "A < B"=@n248))
   B=ScalarParameter() {defaultValue=31.8997, group=Material_Layer_TopBottom, parameterName=HeightBlend_Distance}
  )
  n250 = MaterialFunctionCall(In=Clamp(None=OneMinus(None=@n249)), Contrast=ScalarParameter() {group=Material_Layer_TopBottom, parameterName=HieghtblendContrast}) {materialFunction=&CheapContrast}
  n251 = Clamp(None=Add(A=@n241, B=@n250.Result))
  n252 = MaterialFunctionCall!(
   Base Material=@n219
   Top Material=MaterialFunctionCall!(
    Material=MakeMaterialAttributes!(
     BaseColor=@n226
     Metallic
     Specular=@n229
     Roughness=@n230
     Anisotropy
     EmissiveColor
     Opacity
     OpacityMask
     Normal=@n232
     Tangent
     WorldPositionOffset
     SubsurfaceColor
     ClearCoat
     ClearCoatRoughness
     AmbientOcclusion
     Refraction
     CustomizedUVs_0
     CustomizedUVs_1
     CustomizedUVs_2
     CustomizedUVs_3
     CustomizedUVs_4
     CustomizedUVs_5
     CustomizedUVs_6
     CustomizedUVs_7
     PixelDepthOffset
     ShadingModel
     Displacement
    )
    Tint=LinearInterpolate(A=@n233.RGB, B=@n234.RGB, Alpha=@n239)
    TintMask=Constant() {r=1}
   ) {materialFunction=/Engine/Functions/MaterialLayerFunctions/MatLayerBlend_Tint}.Blended Material
   Alpha=StaticSwitchParameter(
    True=StaticSwitchParameter(True=@n251, False=@n250.Result) {group=Material_Layer_TopBottom, parameterName=UseTopBlendMaterial_Layer2}
    False=@n241
   ) {group=Material_Layer_TopBottom, parameterName=UseGridbasedGroundblend}
  ) {materialFunction=/Game/Environments/Sets/Arid/Materials/MatLayerBlend_StandardCorrectNormals}
  n253 = StaticSwitchParameter(
   True=@n252.Blended Material
   False=StaticSwitchParameter(True=@n252.Blended Material, False=@n219) {group=Material_Layer_TopBottom, parameterName=UseGridbasedGroundblend}
  ) {group=Material_Layer_TopBottom, parameterName=UseTopBlendMaterial_Layer2}
  n254 = Multiply(A=@n253, B=@n168.RGB)
  n255 = VectorParameter() {group=CustomDamage, parameterName=Custom Damage Diffuse Color, defaultValue=<rgba 0 0 0 1>}
  n256 = Multiply(A=@n128, B=ScalarParameter() {defaultValue=0.4, group=CustomDamage, parameterName=Custom Damage DSSm Height Max}) {constB=0.1}
  n257 = Clamp(None=Multiply(A=MaterialFunctionCall(Transition Phase=@n256, Height Texture=@n131) {materialFunction=&HeightLerp}.Lerp Alpha No Contrast) {constB=10})
  n258 = StaticSwitchParameter(
   True=LinearInterpolate(
    A=@n253
    B=StaticSwitchParameter(True=@n254, False=@n255.RGB) {group=CustomDamage, parameterName=Use Custom Damage Diffuse Texture}
    Alpha=@n257
   ) {constB=0, desc=1 is damaged}
   False=@n253
  ) {desc=d, group=CustomDamage, parameterName=Use Custom Damage}
  n259 = MaterialFunctionCall!(
   Manipulate Texture UVs=StaticBoolParameter() {group=Emissive, parameterName=Manipulate Emissive Uvs}
   ActorPositionAdd3
   Mask Cropped Emissive=StaticBoolParameter() {defaultValue=true, group=Emissive, parameterName="Mask Cropped Emissive "}
   Texture=TextureObjectParameter() {group=Emissive, parameterName=EmissiveTexture}
   Animated emissive black out time=ScalarParameter() {defaultValue=0.5, desc=should be a value from 0-1, group=Emissive, parameterName=Animated Emissive Blackout Ratio}
   Texture Multiplier=StaticSwitchParameter(True=Multiply(A=@n165.RGB, B=@n166.RGB), False=@n166.RGB) {group=Emissive, parameterName=Emissive Random Color Chart}
   Modulate Texture=StaticBoolParameter() {group=Emissive, parameterName=Modulate Emissive}
   UVPositioning LowerRight=AppendVector(A=@n167.B, B=@n167.B)
   UVPositioning UpperLeft=ComponentMask(None=@n167.RGB) {g=true, r=true}
   "------------------- "
   "Animate Using Texture Based Animation(T) or a Flicker(F)"=StaticBoolParameter() {group=Emissive, parameterName=AnimateEmissiveUsingAGradient}
   Animate=StaticBoolParameter() {group=Emissive, parameterName=AnimateEmissive}
   -------------------
   "World Position Animation Offset Multiplier (WS)"=ScalarParameter() {defaultValue=0.01, group=Emissive, parameterName=Animated emissive world position animation offset amount}
   Use World Position to offset the animation=StaticBoolParameter() {group=Emissive, parameterName=Use World position to offset the emissive animation}
   Animated Radius=ScalarParameter() {defaultValue=0.1, group=Emissive, parameterName=Emissive Timeline Animation Falloff Radius}
   Animated Hardness=ScalarParameter() {defaultValue=0.1, group=Emissive, parameterName=Emissive Timeline Animation Falloff Hardness}
   "Loop(T) or Ping  Pong(F) Animation"=StaticBoolParameter() {group=Emissive, parameterName="Animate Emissive Using a loop(T) or Ping Pong(F)"}
   Modulate with diffuse=StaticBoolParameter() {group=Emissive, parameterName=Modulate Emissive with Diffuse}
   Diffuse=@n258
   ---------------------
   ControlTimeManually=StaticBoolParameter() {group=Emissive, parameterName=Emissive 1 - Manually Control Emissive Time}
   ManualTime=ScalarParameter() {group=Emissive, parameterName=Emissive 1 - Manual Emissive Time Control}
   Sequence animation speed=ScalarParameter() {defaultValue=0.25, group=Emissive, parameterName=EmissiveAnimationSpeed}
  ) {materialFunction=/Game/Packages/Fortress_SharedMaterials/Base_Material_Functions/SequenceAnimation}
  n260 = StaticSwitchParameter(True=Constant(), False=@n259.Result) {group=Emissive2, parameterName=Use Emissive 2_Exclusive}
  n261 = ActorPositionWS()
  n262 = TextureSampleParameter2D(UVs=Divide(
    A=MaterialFunctionCall(f3=@n261) {materialFunction=&VectorComponentOps}.f3 Add
    B=ScalarParameter() {defaultValue=16.71, group=Emissive2, parameterName="Emissive Color Change Frequency (WS) 2"}
   )) {group=Emissive2, parameterName=Emissive Color Chart 2, texture=&T_EmissiveColorChart}
  n263 = VectorParameter() {group=Emissive2, parameterName=Emissive Multiplier 2, defaultValue=<rgba 1 1 1 1>}
  n264 = VectorParameter() {group=Emissive2, parameterName="Emissive 2 UV Positioning (RG)UpperLeft (BA)LowerRight", defaultValue=<rgba 0 0 1 1>}
  n265 = MaterialFunctionCall(
   Manipulate Texture UVs=StaticBoolParameter() {group=Emissive2, parameterName=Manipulate Emissive 2 Uvs}
   Mask Cropped Emissive=StaticBoolParameter() {defaultValue=true, group=Emissive2, parameterName=Mask Cropped Emissive 2}
   Texture=TextureObjectParameter() {group=Emissive2, parameterName=Emissive Texture 2}
   Animated emissive black out time=ScalarParameter() {defaultValue=0.5, desc=should be a value from 0-1, group=Emissive2, parameterName=Animated Emissive Blackout Ratio 2}
   Texture Multiplier=StaticSwitchParameter(True=Multiply(A=@n262.RGB, B=@n263.RGB), False=@n263.RGB) {group=Emissive2, parameterName=Emissive Random Color Chart2}
   Modulate Texture=StaticBoolParameter() {group=Emissive2, parameterName=Modulate Emissive 2}
   UVPositioning LowerRight=AppendVector(A=@n264.B, B=@n264.B)
   UVPositioning UpperLeft=ComponentMask(None=@n264.RGB) {g=true, r=true}
   "Animate Using Texture Based Animation(T) or a Flicker(F)"=StaticBoolParameter() {group=Emissive2, parameterName=Animate Emissive Using the emissive alpha as a timeline 2}
   Animate=StaticBoolParameter() {group=Emissive2, parameterName=Animate Emissive 2}
   "World Position Animation Offset Multiplier (WS)"=ScalarParameter() {defaultValue=0.01, group=Emissive2, parameterName=Animated emissive world position animation offset amount 2}
   Use World Position to offset the animation=StaticBoolParameter() {group=Emissive2, parameterName=Use World position to offset the emissive animation 2}
   Animated Radius=ScalarParameter() {defaultValue=0.1, group=Emissive2, parameterName=Emissive Timeline Animation Falloff Radius 2}
   Animated Hardness=ScalarParameter() {defaultValue=0.1, group=Emissive2, parameterName=Emissive Timeline Animation Falloff Hardness 2}
   "Loop(T) or Ping  Pong(F) Animation"=StaticBoolParameter() {group=Emissive2, parameterName="Animate Emissive Using a loop(T) or Ping Pong(F) 2"}
   Modulate with diffuse=StaticBoolParameter() {group=Emissive2, parameterName=Modulate Emissive with Diffuse 2}
   Diffuse=@n258
   ControlTimeManually=StaticBoolParameter() {group=Emissive2, parameterName=Emissive 2 - Manually Control Emissive Time}
   ManualTime=ScalarParameter() {group=Emissive2, parameterName=Emissive 2 - Manual Emissive Time Control}
   Sequence animation speed=ScalarParameter() {defaultValue=0.25, group=Emissive2, parameterName=Emissive Animation Speed 2}
  ) {materialFunction=&SequenceAnimation}
  n266 = If(
   A=Constant() {r=1}
   B=@n163
   "A > B"=@n260
   "A == B"=@n260
   "A < B"=If(A=Constant() {r=2}, B=@n163, "A > B"=@n265.Result, "A == B"=@n265.Result, "A < B"=Constant())
  )
  n267 = Add(A=@n259.Result, B=@n265.Result)
  n268 = If(A=Constant() {r=1}, B=@n163, "A > B"=@n259.Result, "A == B"=@n259.Result, "A < B"=Constant())
  n269 = StaticSwitchParameter(
   True=StaticSwitchParameter(
    True=StaticSwitchParameter(True=@n266, False=@n267) {group=Emissive, parameterName=Isolate Emissive to Material Layers}
    False=StaticSwitchParameter(True=@n268, False=@n259.Result) {group=Emissive, parameterName=Isolate Emissive to Material Layers}
   ) {group=Emissive2, parameterName=Use Emissive 2}
   False=Constant()
  ) {group=Emissive, parameterName=Use Emissive}
  n270 = StaticSwitchParameter(
   True=Multiply(A=MaterialFunctionCall!(In) {materialFunction=/Game/Packages/PBW/BASE/Materials/BASE_Animation}.BASE_EmissiveOutput, B=@n269)
   False=@n269
  ) {group=Special Cases, parameterName=Use BASE Emissive Animation And Levels}
  n271 = Fresnel(ExponentIn=ScalarParameter() {defaultValue=0.5, group=Emissive, parameterName=EmissveFresnelExponent})
  n272 = OneMinus(None=@n271)
  n273 = StaticSwitchParameter(True=Multiply(A=@n270, B=StaticSwitchParameter(True=@n272, False=@n271) {group=Emissive, parameterName=InvertEmissiveFresnel}), False=@n270) {group=Emissive, parameterName=Emissive Fresnel}
  n274 = TextureCoordinate() {coordinateIndex=1}
  n275 = ScalarParameter() {defaultValue=1, group=Emissive, parameterName=UVModulateSinePhase}
  n276 = Time()
  n277 = MaterialFunctionCall!(
   Sine Phase=Add(A=ComponentMask(None=@n274) {r=true}, B=Multiply(A=@n275, B=@n276) {constA=1})
   Value 1=ScalarParameter() {defaultValue=1, group=Emissive, parameterName=UVModulateSineValue1}
   Value 2=ScalarParameter() {group=Emissive, parameterName=UVModulateSineValue2}
  ) {materialFunction=/Engine/Functions/Engine_MaterialFunctions02/Utility/Sine_Remapped}
  n278 = MaterialFunctionCall!(
   DiffuseFresnelAmount=Constant() {r=0.2}
   Emissive=StaticSwitchParameter(True=Multiply(A=@n273, B=@n277.Remapped Sine), False=@n273) {group=Emissive, parameterName=ModulateEmissiveByUVS}
   Base Color=@n258
  ) {desc=diffuse, materialFunction=/Game/Packages/Fortress_SharedMaterials/Base_Material_Functions/FresnelFunction}
  n279 = ComponentMask(None=@n278.d) {r=true}
  n280 = Multiply(A=Clamp(None=LinearInterpolate(A=@n157, B=@n158, Alpha=@n160)), B=Clamp(None=LinearInterpolate(A=@n161, B=@n162, Alpha=@n279)))
  n281 = LinearInterpolate(A=@n93, B=ScalarParameter() {defaultValue=1, group=CustomDamage, parameterName=Custom Damage Spec Level}, Alpha=@n257) {constA=1}
  n282 = ComponentMask(None=@n92.Specular Masks) {desc=metalness, g=true}
  n283 = LinearInterpolate(A=@n282, B=ScalarParameter() {defaultValue=10, group=CustomDamage, parameterName=Custom Damage Spec Metallic}, Alpha=@n257)
  n284 = MaterialFunctionCall!(S=ComponentMask(None=@n92.Specular Masks) {b=true, desc=roughness}, V2, V3, V4) {desc=Normals, materialFunction=/Engine/Functions/Engine_MaterialFunctions02/PassThrough}
  n285 = LinearInterpolate(A=@n284.S, B=ScalarParameter() {defaultValue=10, group=CustomDamage, parameterName=Custom Damage Spec Roughness}, Alpha=@n257)
  n286 = MaterialFunctionCall(Float3=MaterialFunctionCall(
    X=StaticSwitchParameter(True=@n281, False=@n93) {group=CustomDamage, parameterName=Use Custom Damage}
    Y=StaticSwitchParameter(True=@n283, False=@n282) {desc=Spec Metalness, group=CustomDamage, parameterName=Use Custom Damage}
    Z=StaticSwitchParameter(True=@n285, False=@n284.S) {desc=Spec Roughness, group=CustomDamage, parameterName=Use Custom Damage}
   ) {desc=Specular Masks, materialFunction=&MakeFloat3}.Result) {materialFunction=&BreakOutFloat3Components}
  n287 = ScalarParameter() {defaultValue=0.1, group=Specular, parameterName=SpecRoughnessMin}
  n288 = ScalarParameter() {defaultValue=0.9, group=Specular, parameterName=SpecRoughnessMax}
  n289 = StaticSwitchParameter(
   True=OneMinus(None=StaticComponentMaskParameter(None=@n170.Out) {defaultR=true, group=Specular, parameterName=DiffuseChannelForRoughness})
   False=@n286.B
  ) {desc=Specular, group=Specular, parameterName=UseDiffuseChannelAsRoughnessMask}
  n290 = Fresnel(ExponentIn=ScalarParameter() {defaultValue=0.68, group=Crystal, parameterName=Core Fresnel EXP})
  n291 = VectorParameter() {group=Crystal, parameterName=Crystal Core Color, defaultValue=<rgba 0.995 0 0.3367 0>}
  n292 = Multiply(
   A=Power(Base=OneMinus(None=@n290), Exp=ScalarParameter() {defaultValue=1, group=Crystal, parameterName=Core Fresnel Power})
   B=Multiply(A=@n291.RGB, B=@n291.RGB)
  )
  n293 = TextureSampleParameter2D() {group=reflection, parameterName=ReflectionMask, samplerSource=SSM_Wrap_WorldGroupSettings}
  n294 = TextureCoordinate() {uTiling=0.5, vTiling=0.5}
  n295 = ScalarParameter() {defaultValue=2.5, group=Crystal, parameterName=MagicPan Scale}
  n296 = Time()
  n297 = ScalarParameter() {defaultValue=0.1, group=Crystal, parameterName=Crystal Magic Time}
  n298 = Rotator(Coordinate=Panner(Coordinate=Multiply(A=@n294, B=@n295), Time=Multiply(A=@n296, B=@n297) {constB=0.25}) {speedX=0.01}) {speed=0.01}
  n299 = ScalarParameter() {defaultValue=1, group=z_code, parameterName=SuperChargedFlickerBias}
  n300 = ScalarParameter() {group=z_code, parameterName=SuperChargedEffect}
  n301 = StaticBool()
  n302 = StaticSwitchParameter(
   True=Multiply(
    A=Multiply(A=@n292, B=@n293.R)
    B=TextureSampleParameter2D(UVs=@n298) {group=reflection, parameterName=ReflectionMask, samplerSource=SSM_Wrap_WorldGroupSettings}.B
   )
   False=StaticSwitchParameter(
    True=MaterialFunctionCall!(AverageHandPosition, SuperChargedFlickerBias=@n299, SuperChargedEffect=@n300, UseDistanceCalc=@n301, Emissive=@n278.Emissive) {materialFunction=/Game/Characters/Fortress_CharacterShader/SuperChargedGlow}.Result
    False=@n278.Emissive
   ) {desc=emissive, group=Special Cases, parameterName=Is a Melee Weapon}
  ) {group=Special Cases, parameterName=is Crystal}
  n303 = CollectionParameter() {
   collection=&FortniteMaterialParameters
   parameterName=FortniteActiveTimeOfDay
  }
  n304 = VectorParameter() {group=Emissive Autoexposure, parameterName=Emissive multiplier over day phases, defaultValue=<rgba 1 1 1 1>}
  n305 = DotProduct(A=@n303, B=AppendVector(A=@n304.RGB, B=@n304.RGB))
  n306 = Constant() {r=1}
  n307 = TextureCoordinate()
  n308 = VectorParameter() {group=Mask, parameterName=Opacity Mask Offset RG Tiling BA, defaultValue=<rgba 0 0 1 1>}
  n309 = Add(A=@n307, B=ComponentMask(None=@n308.RGB) {g=true, r=true})
  n310 = AppendVector(A=@n308.B, B=@n308.B)
  n311 = TextureSampleParameter2D(UVs=StaticSwitchParameter(True=Multiply(A=@n309, B=@n310), False=@n307) {group=Mask, parameterName=Manipulate Opacity Mask Uvs}) {group=Mask, parameterName=OpacityMask, samplerSource=SSM_Wrap_WorldGroupSettings, texture=&T_WhitesRGB}
  n312 = StaticSwitchParameter(
   True=ComponentMask(None=@n170.Out) {a=true}
   False=StaticComponentMaskParameter(None=AppendVector(A=@n311.RGB, B=@n311.RGB)) {defaultR=true, group=Mask, parameterName=Opacity Mask Channel}
  ) {group=Mask, parameterName="Use Diffuse Texture Alpha As Opacity Mask (Masked Instances Only)"}
  n313 = Constant() {r=2}
  n314 = StaticSwitchParameter(
   True=ShadowReplace(
    Default=MaterialFunctionCall(Alpha Threshold=@n312) {materialFunction=&DitherTemporalAA}.Result
    Shadow=@n312
   )
   False=StaticSwitchParameter(
    True=MaterialFunctionCall(In=@n312, Contrast=@n313) {materialFunction=&CheapContrast}.Result
    False=@n312
   ) {group=Mask, parameterName=Distance Field Opacity Mask Texture?}
  ) {group=Mask, parameterName=Dither Opacity Mask}
  n315 = Subtract(A=ComponentMask(None=TextureCoordinate() {coordinateIndex=1}) {r=true}, B=ScalarParameter() {parameterName=Opacity Mask Limit to Layer X})
  n316 = Multiply(A=Ceil(None=@n315), B=OneMinus(None=Floor(None=@n315)))
  n317 = TextureSampleParameter2D() {group=reflection, parameterName=ReflectionMask, samplerSource=SSM_Wrap_WorldGroupSettings}
  n318 = Fresnel(ExponentIn=ScalarParameter() {defaultValue=0.4, group=SubSurface, parameterName=SS Exponent In})
  n319 = VectorParameter() {group=SubSurface, parameterName=SSS Color, defaultValue=<rgba 1 1 1 0>}
  n320 = ComponentMask(None=Transform(None=ReflectionVectorWS()) {transformSourceType=TRANSFORMSOURCE_World, transformType=TRANSFORM_View}) {g=true, r=true}
  n321 = TextureSampleParameter2D(UVs=Multiply(A=ConstantBiasScale(None=@n320), B=Constant2Vector() {g=-1, r=1})) {group=reflection, parameterName=SphereReflectionMap, samplerSource=SSM_Wrap_WorldGroupSettings}
  n322 = Multiply(A=Power(Base=OneMinus(None=@n318)), B=Multiply(A=Multiply(A=@n319.RGB, B=@n321.RGB), B=@n319.A))
  n323 = TextureSampleParameter2D() {
   group=Specular
   parameterName=SpecularMasks
   samplerSource=SSM_Wrap_WorldGroupSettings
   samplerType=SAMPLERTYPE_Masks
   texture=&T_Fortnite_Default_S
  }
  n324 = MakeMaterialAttributes(
   BaseColor=StaticSwitchParameter(True=LinearInterpolate(A=@n278.d, B=@n156.RGB, Alpha=@n280), False=@n278.d) {parameterName=UseSnow}
   Metallic=StaticSwitchParameter(True=@n286.G, False=Constant()) {defaultValue=true, desc=Specular, group=Specular, parameterName=Use Metallic}
   Specular=StaticSwitchParameter(True=@n286.R, False=Constant()) {defaultValue=true, desc=Specular, group=Specular, parameterName=Use Spec}
   Roughness=StaticSwitchParameter(True=LinearInterpolate(A=@n287, B=@n288, Alpha=@n289), False=Constant() {r=1}) {defaultValue=true, desc=Specular, group=Specular, parameterName=Use Roughness}
   EmissiveColor=StaticSwitchParameter(True=Multiply(A=@n302, B=@n305), False=@n302) {group=Emissive Autoexposure, parameterName=Use Simple Emissive Exposure}
   Opacity=StaticSwitchParameter(True=ScalarParameter() {group=SubSurface, parameterName=SS Amount}, False=Constant() {r=1}) {group=Special Cases, parameterName=is Crystal}
   OpacityMask=StaticSwitchParameter(True=LinearInterpolate(A=@n306, B=@n314, Alpha=@n316), False=@n314) {group=Mask, parameterName=Limit OpacityMask to SingleLayer}
   Normal=@n159
   WorldPositionOffset=@n151
   SubsurfaceColor=StaticSwitchParameter(True=Multiply(A=@n317.R, B=@n322), False=Constant()) {group=Special Cases, parameterName=is Crystal}
   ClearCoat=StaticSwitchParameter(True=@n323.A, False=ScalarParameter() {defaultValue=1, group=Clear Coat, parameterName=Clear Coat Opacity, sliderMax=1}) {defaultValue=true, group=Clear Coat, parameterName=Clear Coat Mask}
   ClearCoatRoughness=ScalarParameter() {group=Clear Coat, parameterName=Clear Coat Roughness}
  )
  n325 = MaterialAttributeLayers!(None=@n324)
  n326 = StaticSwitchParameter(
   True=MakeMaterialAttributes(EmissiveColor=Multiply(A=@n97, B=@n98.RGB), Opacity=Multiply(A=@n99, B=@n116), WorldPositionOffset=Add(A=@n152, B=@n155))
   False=MaterialFunctionCall!(Environment FX=StaticSwitchParameter(True=@n325, False=@n324) {parameterName=Use Material Layer 1}) {materialFunction=/Game/Packages/Fortress_SharedMaterials/Base_Material_Functions/Environment_Effects}.Result
  ) {group=Special Cases, parameterName=Light Cone Effect}
  n327 = StaticSwitchParameter(
   True=MaterialFunctionCall!(In=@n326, UVs, TransitionGradient) {materialFunction=/Game/Environments/Sets/Vulcano/Materials/MF_LavaEdgeGradientSimple}.Result
   False=@n326
  ) {parameterName=UseLavaEdgeGradient}
  n328 = GetMaterialAttributes(None=@n327) {attributeGetTypes [#1]}
  n329 = SetMaterialAttributes!(
   MaterialAttributes=@n328.MaterialAttributes
   Subsurface Color=Multiply(A=@n328.Base Color, B=VectorParameter() {group=SimpleSS, parameterName=SimpleSS_Color, defaultValue=<rgba 2 2 2 1>}.RGB)
   Opacity=ScalarParameter() {defaultValue=0.5, group=SimpleSS, parameterName=SimpleSS_Opacity}
   Shading Model=ShadingModel() {shadingModel=MSM_TwoSidedFoliage}
  ) {attributeSetTypes [#2, #3, #4]}
  n330 = MaterialFunctionCall!(
   In=MaterialFunctionCall!(In=StaticSwitchParameter(True=@n329, False=@n327) {group=SimpleSS, parameterName=UseSSShading}) {materialFunction=/Game/Packages/Fortress_SharedMaterials/Base_Material_Functions/MF_ChestAnimation}.Result
   UseSphericalMappingOverride
  ) {materialFunction=/Game/Packages/Fortress_SharedMaterials/Base_Material_Functions/MF_CurieFallbackEffect}
  n331 = ScalarParameter() {bUseCustomPrimitiveData=true, group=z_GameplayParameters_DestructibleDamage, parameterName=Code_DestructionAmount, primitiveDataIndex=13}
  n332 = MaterialFunctionCall(
   Material_In=MaterialFunctionCall!(In=MaterialFunctionCall(MA In=@n330.Result) {materialFunction=&MF_CreativeMoveTool}.MA Out) {materialFunction=/Game/Athena/WinterArt/Materials/MF_Events}.Result
   UseDamageWobbleWPO=StaticBool()
   DestructibleDamageAmount=@n331
  ) {materialFunction=&MF_FN_DestructibleDamage}
  NamedRerouteDeclaration(None=@n332.DamageAnimWPO) {name=DamageAnimWPO, nodeColor=<rgba 0 0.3437 1 1>}
  PerInstanceRandom() {desc=TODO: maybe implement this later}
  n333 = Constant() {r=3}
  n334 = If(
   A=Constant() {r=2}
   "A > B"=If(
    A=Constant() {r=1}
    "A > B"=ScalarParameter() {defaultValue=1, group=WS Gradient, parameterName=DustL1, sortPriority=4}
    "A < B"=ScalarParameter() {group=WS Gradient, parameterName=DustL2, sortPriority=5}
   ) {desc=0}
   "A < B"=ScalarParameter() {group=WS Gradient, parameterName=DustL3, sortPriority=6}
  ) {constB=0.15}
  n335 = ScalarParameter() {group=WS Gradient, parameterName=DustL4, sortPriority=7}
  n336 = ScalarParameter() {defaultValue=0.2, parameterName=DustSides}
  n337 = ScalarParameter() {defaultValue=0.45, parameterName=DustUp Max}
  n338 = ComponentMask(None=VertexNormalWS()) {b=true}
  Saturate(None=Multiply(A=If(A=@n333, "A > B"=@n334, "A < B"=@n335), B=LinearInterpolate(A=@n336, B=@n337, Alpha=@n338)))
  ScalarParameter() {defaultValue=0.75, group=Diffuse, parameterName=BaseColorFresnelAmount}
  ScalarParameter() {defaultValue=0.2, group=Specular, parameterName=SpecLevel}
  ScalarParameter() {group=z_code, parameterName=WS Gradient Top Distance From Bottom}
  ScalarParameter() {group=z_code, parameterName=WS Gradient World space Bottom Z Position}
  ScalarParameter() {defaultValue=0.5, desc="0= Full material replace 1= blend angle corrected normals", group=z_code, parameterName=Code_Decal_NormalBlendMethod}
  ScalarParameter() {group=Decal, parameterName=Code_DecalParametersSet}
  ScalarParameter() {defaultValue=1, group=z_code, parameterName=CODE_GradientModulation}
  ScalarParameter() {defaultValue=0.5, group=z_code, parameterName=Code_Decal_MaskingCheapContrast}
  n339 = GetMaterialAttributes(None=@n332.Material_Out) {attributeGetTypes [#5]}
  SetMaterialAttributes!(
   MaterialAttributes=@n339.MaterialAttributes
   World Position Offset=LinearInterpolate(B=@n339.World Position Offset, Alpha=ScalarParameter() {defaultValue=1, parameterName=BP_UseSign}) {constA=9999999827968}
  ) {attributeSetTypes [#5]}
  StaticBool()
  StaticSwitchParameter() {group=Decal, parameterName=Use Local Space Decal}
  StaticSwitchParameter() {group=Decal, parameterName=Support Local Space Decal Normals}
  TextureObjectParameter() {group=z_code, parameterName=Code_DecalTexture_Normals}
  TextureObjectParameter() {
   group=Wind Textures
   parameterName=X-Axis Texture And X Extent Alpha
   samplerType=SAMPLERTYPE_LinearColor
   texture=/Game/Global/Textures/Default/Blanks/T_BlackVectorDisplacement
  }
  TextureObjectParameter() {
   group=Wind Textures
   parameterName=Position and Index Texture
   samplerType=SAMPLERTYPE_LinearColor
   texture=/Engine/Functions/Engine_MaterialFunctions02/PivotPainter2/Black_1x1_EXR_Texture
  }
  TextureObjectParameter() {group=z_code, parameterName=WS Gradient}
  TextureSampleParameter2D() {group=z_code, parameterName=Code_DecalTexture_BaseColor, samplerSource=SSM_Clamp_WorldGroupSettings}
  TextureSampleParameter2D() {
   group=z_code
   parameterName=Code_DecalTexture_Spec
   samplerSource=SSM_Clamp_WorldGroupSettings
   samplerType=SAMPLERTYPE_Masks
   texture=/Game/Global/Textures/Default/Blanks/Temp_specular_Mask
  }
  VectorParameter() {
   desc="1,0,0,0 applies the ws gradient to the first \"material\" and none of the others. 1,1,0,0 applies the gradient to the first two materials."
   group=z_code
   parameterName=CODE_TextureGroupsToApplyGradientsTo
   defaultValue=<rgba 1 0 1 0>
  }
  VectorParameter() {group=z_code, parameterName=Code_DecalLocalNormal, defaultValue=<rgba 0 0 0 1>}
  VectorParameter() {group=z_code, parameterName=Code_DecalLocalPosition, defaultValue=<rgba 0 0 0 1>}
  VectorParameter() {group=z_code, parameterName=Code_DecalLocalTangent, defaultValue=<rgba 0 0 0 1>}
  VectorParameter() {group=z_code, parameterName=Code_DecalLocalBiNormal, defaultValue=<rgba 0 0 0 1>}
  VectorParameter() {group=z_code, parameterName=Code_DecalScale, defaultValue=<rgba 1 1 1 1>}
  VectorParameter() {group=z_code, parameterName=CODE_TextureGroupsToApplyCustomDecalsTo, defaultValue=<rgba 1 1 1 1>}
  VertexInterpolator(VS=@n5)
 ]
}
